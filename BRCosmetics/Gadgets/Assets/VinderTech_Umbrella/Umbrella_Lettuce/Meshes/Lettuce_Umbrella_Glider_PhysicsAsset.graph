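# Physics asset for the Lettuce Umbrella glider mesh, a VinderTech umbrella gadget cosmetic.
# It approximates the umbrella with two capsule bodies: a long, narrow one on the Melee_TwoHanded_Handle bone and a wide, nearly spherical one on blade_01.
# A single constraint joins blade_01 to the handle, with matching angle values set for both swing axes and twist.
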PhysicsAsset /BRCosmetics/Gadgets/Assets/VinderTech_Umbrella/Umbrella_Lettuce/Meshes/Lettuce_Umbrella_Glider_PhysicsAsset
bodies {
 Melee_TwoHanded_Handle [{shapeType=Capsule, length=66.8597, radius=10.1419, center=<0.2943 0.238 48.5656>, rotation=<rot 0.0989 -90 0.1195>}]
 blade_01 [{shapeType=Capsule, length=62.784, radius=61.6538, center=<0.1184 0.0117 -6.916>, rotation=<rot 0.0495 96.1653 -90>}]
}
constraints [{bone1Name=blade_01, bone2Name=Melee_TwoHanded_Handle, swing1LimitDegrees=45, swing2LimitDegrees=45, twistLimitDegrees=45}]
